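# Test light function material for the Festivus environment, with several branches that react to a smoothed FFT value from a parameter collection.
# Branches include an animated orbit flipbook, two panning fire textures blended with a head texture and a screen-pixel mask, and a single-point media texture sample.
# The FFT value drives brightness and blending.
# The head's scaling math cancels to a fixed center UV, and no branch is shown wired to a material output.
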Material /Game/Athena/Environments/Festivus/Materials/FFT_Materials/M_FFT_Test_02_LF
graph {
 properties {materialDomain=MD_LightFunction, pixelDepthOffsetMode=PDOM_Legacy, shadingModel=MSM_Unlit, usedShadingModels=Unlit}
 nodes [
  n1 = CollectionParameter() {collection=&/FFT_Breakout_MCP, parameterName=FFT_01_Smothed}
  Add(B=@n1) {constA=-0.2}
  Constant3Vector() {constant=<rgba 1 1 1 0>}
  n2 = Time()
  n3 = Constant() {r=16}
  n4 = TextureObject() {texture=/Game/Athena/Environments/Festivus/Materials/T_MKartOrbit_01}
  n5 = TextureSample(
   UVs=MaterialFunctionCall("Animation  Phase (0-1)"=Multiply(A=@n2) {constB=0.075}, Number of Rows=@n3, Number of Columns=@n3, Texture=@n4, MipBias/Level=Constant()) {materialFunction=&FlipBook}.UVs
   Tex=@n4
  ) {texture=&T_MKartOrbit_01}
  n6 = Multiply(A=@n1) {constB=2}
  Multiply(A=@n5.RGB, B=Clamp(None=@n6))
  n7 = TextureCoordinate()
  n8 = TextureSample(UVs=Panner(Coordinate=@n7, Time=Time()) {speedX=0.03, speedY=0.5}) {texture=/Game/Athena/Environments/Festivus/Materials/T_Fire_01_D}
  n9 = TextureSample(UVs=Panner(Coordinate=@n7, Time=Time()) {speedX=-0.15, speedY=0.25}) {texture=&T_Fire_01_D}
  n10 = Constant2Vector() {g=0.5, r=0.5}
  n11 = CollectionParameter() {collection=&FFT_Breakout_MCP, parameterName=FFT_01_Smothed}
  n12 = OneMinus(None=Clamp(None=Multiply(A=@n11) {constB=0.2}) {maxDefault=5})
  n13 = Divide(A=@n12)
  n14 = Divide(A=@n13, B=@n12)
  n15 = TextureSample(UVs=Subtract(A=Multiply(A=@n10, B=@n12), B=Subtract(A=@n13, B=@n14))) {texture=/Game/Athena/Environments/Festivus/Materials/T_MHead_01}
  n16 = LinearInterpolate(A=Multiply(A=Multiply(A=@n8.RGB, B=@n9.RGB)) {constB=5}, B=@n15.RGB, Alpha=@n15.RGB)
  n17 = TextureCoordinate() {uTiling=96, vTiling=96}
  Multiply(A=Multiply(A=@n16, B=TextureSample(UVs=@n17) {texture=/Game/Athena/Environments/Festivus/Materials/T_ScreenPixels_M_01}.R)) {constB=10}
  Multiply(A=LinearInterpolate(A=@n5.RGB, B=@n16, Alpha=@n6), B=@n1)
  Multiply(
   A=TextureSample(UVs=Constant2Vector() {g=0.5, r=0.5}) {texture=/Game/Athena/Environments/Festivus/Materials/Video/Festivus_MediaTexture}.RGB
   B=@n1
  )
  ScalarParameter() {defaultValue=0.4429, parameterName=Scale}
  TextureCoordinate()
  TextureSample() {texture=/Game/Athena/Environments/Festivus/Materials/RT_Screens_01}
 ]
}
